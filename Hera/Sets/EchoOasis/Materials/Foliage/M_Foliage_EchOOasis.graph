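# Two-sided subsurface foliage material for the Echo Oasis set, supporting Nanite and instanced meshes.
# A shared foliage color function feeds snow and medium-plant animation, then an impostor-capture switch that is off and the Curie fallback effect.
# An Echo Oasis sphere-mask hue function blends into base color by parameter before the Creative move-tool function.
# A separate debug branch, noted for local connection, tints or hides overlapping instances using per-instance data.
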Material /Hera/Sets/EchoOasis/Materials/Foliage/M_Foliage_EchOOasis
graph {
 properties {
  bUseMaterialAttributes=true
  bUsedWithInstancedStaticMeshes=true
  bUsedWithNanite=true
  pixelDepthOffsetMode=PDOM_Legacy
  shadingModel=MSM_Subsurface
  twoSided=true
  usedShadingModels=Subsurface
 }
 nodes [
  n1 = StaticBool() {desc=Don't change, value=true}
  n2 = Constant() {r=1}
  n3 = MaterialFunctionCall!(UseSpecMaskG_AsOpacityMask) {materialFunction=/Game/Global/Base4Layers_2022/MaterialFunctions/MF_FoliageMaster2023Colors}
  n4 = StaticBoolParameter() {group=Snow, parameterName=SnowFlipTwosidedNormal}
  n5 = StaticBool() {value=true}
  n6 = StaticBool() {value=true}
  n7 = MaterialFunctionCall!(In=NamedRerouteDeclaration(None=MaterialFunctionCall!(In=@n3.Result, FlipTwoSidedSign=@n4, IsTangentSpaceInput=@n5, HasBackFlipFaceNormals=@n6) {materialFunction=/Game/Packages/Fortnite_UniqueMaterials_1/MaterialFunctions/Trees/MF_Tree_Snow}.Result) {name=Just The Plant, nodeColor=<rgba 1 0 0.7266 1>}) {materialFunction=/Game/Environments/Asteria/Foliage/Plants/BaseMaterial/MF_MediumPlantAnimation}
  n8 = StaticBool()
  n9 = MaterialFunctionCall!(
   In=StaticSwitch(
    True=MaterialFunctionCall!(Material Attributes=@n7.Result, IsTangentSpaceNormal=@n1, In=@n2) {materialFunction=/Game/ImpostorBaker/MaterialFunctions/ImposterCaptureSwitch}.Result
    False=@n7.Result
    Value=NamedRerouteDeclaration(None=@n8) {name=IsImpostorBake, nodeColor=<rgba 0 1 1 1>}
   )
   UseSphericalMappingOverride
  ) {materialFunction=/Game/Packages/Fortress_SharedMaterials/Base_Material_Functions/MF_CurieFallbackEffect}
  n10 = GetMaterialAttributes(None=@n9.Result) {attributeGetTypes [#1]}
  n11 = MaterialFunctionCall!(Aditional Mask, Radius, Hardness, Base Color=@n10.Base Color, Location, Height?, Color Picker, Subsurface, Multiply HUE) {materialFunction=&/MF_EchoOasis_SphereMaskHUE}
  n12 = ScalarParameter() {parameterName=use Echooasis?, sliderMax=1}
  MaterialFunctionCall(MA In=SetMaterialAttributes!(MaterialAttributes=@n10.MaterialAttributes, Base Color=LinearInterpolate(A=@n10.Base Color, B=@n11.New Base Color, Alpha=@n12)) {attributeSetTypes [#1]}) {materialFunction=&MF_CreativeMoveTool}
  n13 = PerInstanceCustomData!(DefaultValue)
  n14 = ScalarParameter() {defaultValue=1, parameterName=ShowOverlapColor}
  n15 = PerInstanceCustomData()
  SetMaterialAttributes!(
   MaterialAttributes=@n9.Result
   Emissive Color=LinearInterpolate(B=VectorParameter() {parameterName=BadColor, defaultValue=<rgba 0 1 0.5746 1>}.RGB, Alpha=Multiply(A=@n13, B=@n14))
   World Position Offset=LinearInterpolate(B=LinearInterpolate(Alpha=@n15) {constB=150000}, Alpha=ScalarParameter() {parameterName=HideOverlaps})
   Opacity Mask
  ) {
   desc=Connect me locally to see debug stuff
   attributeSetTypes [#2, #3, #4]
  }
  StaticBoolParameter() {parameterName=UseFlipBackFaceNormals}
 ]
}
